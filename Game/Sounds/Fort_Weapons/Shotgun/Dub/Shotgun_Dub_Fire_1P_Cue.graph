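# First-person fire sound cue for the Dub shotgun, routed to the first-person ranged-fire sound class with gunfire concurrency and first-person projectile attenuation.
# Each quality branch mixes a main shot, a briefly delayed fire layer pitched randomly downward, and a later pushback sound.
# The first quality level picks the main shot from three random variations, while the other levels share a branch with one fixed shot.
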SoundCue /Game/Sounds/Fort_Weapons/Shotgun/Dub/Shotgun_Dub_Fire_1P_Cue
root=QualityLevel_0
nodes [
 {id=Delay_0, children [Modulator_1], properties {delayMax=0.02, delayMin=0.02}}
 {id=Delay_3, children [WavePlayer_16], properties {delayMax=0.2, delayMin=0.2}}
 {id=Delay_4, children [Modulator_10], properties {delayMax=0.02, delayMin=0.02}}
 {id=Delay_5, children [WavePlayer_25], properties {delayMax=0.2, delayMin=0.2}}
 {id=Mixer_0, children [Modulator_0,Delay_0,Modulator_3], properties {inputVolume [1,0.8,1]}}
 {id=Mixer_1, children [Modulator_6,Delay_4,Modulator_13], properties {inputVolume [1,0.8,1]}}
 {id=Modulator_0, children [Random_0]}
 {id=Modulator_1, children [WavePlayer_3], properties {pitchMax=0.8, pitchMin=0.5}}
 {id=Modulator_10, children [WavePlayer_24], properties {pitchMax=0.8, pitchMin=0.5}}
 {id=Modulator_12, children [Mixer_1]}
 {id=Modulator_13, children [Delay_5]}
 {id=Modulator_2, children [Mixer_0]}
 {id=Modulator_3, children [Delay_3]}
 {id=Modulator_6, children [WavePlayer_18]}
 {id=QualityLevel_0, children [Modulator_2,Modulator_12,Modulator_12,Modulator_12]}
 {id=Random_0, children [WavePlayer_0,WavePlayer_1,WavePlayer_2], properties {weights [1,1,1]}}
 {id=WavePlayer_0, properties {soundWaveAssetPtr=&/Shotgun_Dub_Fire_1P_01}}
 {id=WavePlayer_1, properties {soundWaveAssetPtr=&/Shotgun_Dub_Fire_1P_02}}
 {id=WavePlayer_16, properties {soundWaveAssetPtr=&/Shotgun_Dub_Pushback_03}}
 {id=WavePlayer_18, properties {soundWaveAssetPtr=&Shotgun_Dub_Fire_1P_01}}
 {id=WavePlayer_2, properties {soundWaveAssetPtr=&/Shotgun_Dub_Fire_1P_03}}
 {id=WavePlayer_24, properties {soundWaveAssetPtr=&/Shotgun_Dub_FireLayer_01}}
 {id=WavePlayer_25, properties {soundWaveAssetPtr=&Shotgun_Dub_Pushback_03}}
 {id=WavePlayer_3, properties {soundWaveAssetPtr=&Shotgun_Dub_FireLayer_01}}
]
properties {
 attenuationSettings=&Fort_Weapon_Projectile_FirstPerson_Athena_Att
 maxDistance=2097152
 pitchMultiplier=0.9
 soundClassObject=&Fort_WeaponRanged_Fire_1P_PSM
 volumeMultiplier=0.65
 attenuationOverrides {falloffDistance=5000, attenuationShapeExtents {x=1000}}
 concurrencySet [&Weapon_Gunfire_FirstPerson_Conc]
}
