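# Sound cue for the Dub shotgun's gunfire, mixing close, medium and distant layers, each with its own attenuation and sound class.
# The close layer pairs a pitched-down random shot with a slightly delayed, lower-pitched shotgun break sample.
# Medium and distant layers randomize samples through modulators.
# A quality-level switch gives the first level larger sample pools and different medium waves, while the other three share a smaller mix.
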SoundCue /Game/Sounds/Fort_Weapons/Shotgun/Dub/Shotgun_Dub_Fire_Cue
root=QualityLevel_0
nodes [
 {
  id=Attenuation_1
  children [Modulator_1]
  properties {attenuationSettings=&Fort_Weapon_Projectile_Medium_Mono_Athena_Att}
 }
 {id=Attenuation_2, children [Modulator_7], properties {attenuationSettings=&Fort_Weapon_Projectile_Close_Athena_Att}}
 {
  id=Attenuation_3
  children [Modulator_6]
  properties {attenuationSettings=&Fort_Weapon_Projectile_Distant_Mono_Athena_Att}
 }
 {id=Attenuation_4, children [Modulator_11], properties {attenuationSettings=&Fort_Weapon_Projectile_Close_Athena_Att}}
 {
  id=Attenuation_5
  children [Modulator_10]
  properties {attenuationSettings=&Fort_Weapon_Projectile_Distant_Mono_Athena_Att}
 }
 {
  id=Attenuation_6
  children [Modulator_2]
  properties {attenuationSettings=&Fort_Weapon_Projectile_Medium_Mono_Athena_Att}
 }
 {id=Delay_1, children [Modulator_5], properties {delayMax=0.02, delayMin=0.02}}
 {id=Delay_2, children [Modulator_9], properties {delayMax=0.02, delayMin=0.02}}
 {id=Mixer_2, children [Modulator_4,Delay_1], properties {inputVolume [1,1]}}
 {id=Mixer_3, children [SoundClass_2,SoundClass_6,SoundClass_3], properties {inputVolume [1,0.8,0.8]}}
 {id=Mixer_4, children [Modulator_8,Delay_2], properties {inputVolume [1,1]}}
 {id=Mixer_5, children [SoundClass_4,SoundClass_1,SoundClass_5], properties {inputVolume [1,1,1]}}
 {id=Modulator_1, children [Random_0], properties {volumeMax=2.05, volumeMin=1.95}}
 {id=Modulator_10, children [Random_4], properties {pitchMax=1.1}}
 {id=Modulator_11, children [Mixer_4]}
 {id=Modulator_2, children [Random_6]}
 {id=Modulator_4, children [Random_3], properties {pitchMax=0.95, pitchMin=0.9}}
 {id=Modulator_5, children [WavePlayer_15], properties {pitchMax=0.8, pitchMin=0.5}}
 {id=Modulator_6, children [Random_2], properties {pitchMax=1.1}}
 {id=Modulator_7, children [Mixer_2]}
 {id=Modulator_8, children [Random_5], properties {pitchMax=0.95, pitchMin=0.9}}
 {id=Modulator_9, children [WavePlayer_23], properties {pitchMax=0.8, pitchMin=0.5}}
 {id=QualityLevel_0, children [Mixer_3,Mixer_5,Mixer_5,Mixer_5]}
 {id=Random_0, children [WavePlayer_0,WavePlayer_1,WavePlayer_2], properties {weights [1,1,1]}}
 {id=Random_2, children [WavePlayer_8, WavePlayer_9, WavePlayer_10, WavePlayer_11], properties {weights [1,1,1,1]}}
 {id=Random_3, children [WavePlayer_12, WavePlayer_13, WavePlayer_14], properties {weights [1,1,1]}}
 {id=Random_4, children [WavePlayer_16, WavePlayer_17, WavePlayer_19], properties {weights [1,1,1]}}
 {id=Random_5, children [WavePlayer_20, WavePlayer_22], properties {weights [1,1]}}
 {id=Random_6, children [WavePlayer_7, WavePlayer_18], properties {weights [1,1]}}
 {
  id=SoundClass_1
  children [Attenuation_6]
  properties {soundClassOverride=&Fort_WeaponRanged_Fire_Distant}
 }
 {
  id=SoundClass_2
  children [Attenuation_2]
  properties {soundClassOverride=&Fort_WeaponRanged_Fire_3P_PSM}
 }
 {
  id=SoundClass_3
  children [Attenuation_3]
  properties {soundClassOverride=&Fort_WeaponRanged_Fire_Distant}
 }
 {
  id=SoundClass_4
  children [Attenuation_4]
  properties {soundClassOverride=&Fort_WeaponRanged_Fire_3P_PSM}
 }
 {
  id=SoundClass_5
  children [Attenuation_5]
  properties {soundClassOverride=&Fort_WeaponRanged_Fire_Distant}
 }
 {
  id=SoundClass_6
  children [Attenuation_1]
  properties {soundClassOverride=&Fort_WeaponRanged_Fire_Distant}
 }
 {id=WavePlayer_0, properties {soundWaveAssetPtr=/Game/Sounds/Fort_Weapons/Shotgun/weapons_shotguns_bigger_013_MediumNew}}
 {id=WavePlayer_1, properties {soundWaveAssetPtr=/Game/Sounds/Fort_Weapons/Shotgun/weapons_shotguns_bigger_014_MediumNew}}
 {id=WavePlayer_10, properties {soundWaveAssetPtr=/Game/Sounds/Fort_Weapons/Distant/Distant_low_shot_01b}}
 {id=WavePlayer_11, properties {soundWaveAssetPtr=/Game/Sounds/Fort_Weapons/Distant/Distant_low_shot_01a}}
 {id=WavePlayer_12, properties {soundWaveAssetPtr=/Game/Sounds/Fort_Weapons/Shotgun/weapons_shotguns_bigger_013}}
 {id=WavePlayer_13, properties {soundWaveAssetPtr=/Game/Sounds/Fort_Weapons/Shotgun/weapons_shotguns_bigger_014}}
 {id=WavePlayer_14, properties {soundWaveAssetPtr=/Game/Sounds/Fort_Weapons/Shotgun/weapons_shotguns_bigger_015}}
 {id=WavePlayer_15, properties {soundWaveAssetPtr=/Game/Sounds/Fort_Weapons/Shotgun/weapon_shotgun_break2_close_02}}
 {id=WavePlayer_16, properties {soundWaveAssetPtr=/Game/Sounds/Fort_Weapons/Distant/Distant_low_shot_01c}}
 {id=WavePlayer_17, properties {soundWaveAssetPtr=/Game/Sounds/Fort_Weapons/Distant/Distant_low_shot_01d}}
 {id=WavePlayer_18, properties {soundWaveAssetPtr=/Game/Sounds/Fort_Weapons/Shotgun/weapons_shotguns_bigger_014_Medium}}
 {id=WavePlayer_19, properties {soundWaveAssetPtr=&Distant_low_shot_01a}}
 {id=WavePlayer_2, properties {soundWaveAssetPtr=/Game/Sounds/Fort_Weapons/Shotgun/weapons_shotguns_bigger_015_MediumNew}}
 {id=WavePlayer_20, properties {soundWaveAssetPtr=&weapons_shotguns_bigger_013}}
 {id=WavePlayer_22, properties {soundWaveAssetPtr=&weapons_shotguns_bigger_015}}
 {id=WavePlayer_23, properties {soundWaveAssetPtr=&weapon_shotgun_break2_close_02}}
 {id=WavePlayer_7, properties {soundWaveAssetPtr=/Game/Sounds/Fort_Weapons/Shotgun/weapons_shotguns_bigger_013_Medium}}
 {id=WavePlayer_8, properties {soundWaveAssetPtr=&Distant_low_shot_01c}}
 {id=WavePlayer_9, properties {soundWaveAssetPtr=&Distant_low_shot_01d}}
]
properties {
 maxDistance=27500
 volumeMultiplier=0.7
 attenuationOverrides {falloffDistance=5000, attenuationShapeExtents {x=1000}}
 concurrencySet [&Weapon_Gunfire_Local_Conc, &Weapon_Gunfire_Global_Conc]
}
